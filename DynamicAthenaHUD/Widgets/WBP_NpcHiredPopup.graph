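# NPC-hired popup, a child of the backplate notification widget, showing the name of the player's newly hired NPC as now following them.
# On construct, it finds the last NPC-converted pawn in the player pawn's ConvertComponent and compares its PlayerState to the cached one.
# A new hire sets text and icon and plays the intro animation.
# After a delay it rechecks, replaying for another new hire or removing its Dynamic UI scene.
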WidgetBlueprint /DynamicAthenaHUD/Widgets/WBP_NpcHiredPopup
graphs [
 CheckForNewHiredNpc {nodes [
   CheckForNewHiredNpc() {is=title K2Node_FunctionEntry}
   n1 = GetLastHiredNpcPlayerState!() -> (NpcPlayerState:Player State Object Reference)
   > n2 = Branch(Condition="Not Equal (Object)"(A:Object Reference=@n1, B:Object Reference=Get CurrentHiredNpcPlayerState() -> (CurrentHiredNpcPlayerState:Player State Object Reference)) -> (ReturnValue:bool=~) {is=title K2Node_PromotableOperator})
   > Set CurrentHiredNpcPlayerState(CurrentHiredNpcPlayerState:Player State Object Reference=@n1) -> (Output_Get:Player State Object Reference)
   > Return Node!(execute, NewNpcWasHired:bool=true)
   Return Node(execute=@n2.else, NewNpcWasHired=false)
  ]}
 EventGraph {nodes [
   ShowNpcHiredMessage() {is=title K2Node_CustomEvent}
   > SetDetails!(execute, self:WBP Notification With Backplate Visuals Object Reference, InText:Text=Get DetailsText() -> (DetailsText:Text)) -> (then)
   n1 = Get CurrentHiredNpcPlayerState()
   > SetHeadling!(
    execute
    self:WBP Notification With Backplate Visuals Object Reference
    InText:Text="To Text (String)"(InString=GetPlayerName!(self:Player State Object Reference=@n1) -> (ReturnValue:String))
   ) -> (then)
   > SetBrushFromMaterial(
    self:WBP Notification With Backplate Visuals Object Reference
    Material=Get IconMaterial() -> (IconMaterial:Material Instance Object Reference)
   )
   > PlayIntroAnimation(self:WBP Notification With Backplate Visuals Object Reference)
   > Delay(Duration=3.25)
   > n2 = CheckForNewHiredNpc!(execute, self) -> (then, NewNpcWasHired:bool) {is=title K2Node_CallFunction}
   > n3 = Branch(Condition=@n2)
   > ShowNpcHiredMessage() {is=title K2Node_CallFunction}
   RemoveScene!(
    execute=@n3.else
    self:Dynamic UIManager Object Reference=Get DynamicUIManager() -> (ReturnValue:Dynamic UIManager Object Reference) {is=title K2Node_GetSubsystem}
    Scene:Dynamic UIScene Object Reference=Get SceneReference() -> (SceneReference:Dynamic UIScene Object Reference)
    Player:"Player Controller Object Reference (by ref)"=GetOwningPlayer()
   ) -> (then)
   Event Destruct()
   > Parent: Destruct()
   > Set CurrentHiredNpcPlayerState()
   Event Construct()
   > Parent: Construct()
   > n4 = CheckForNewHiredNpc()
   > Branch(Condition=@n4)
   > ShowNpcHiredMessage()
  ]}
 GetLastHiredNpcPlayerState {nodes [
   GetLastHiredNpcPlayerState() {is=title K2Node_FunctionEntry}
   > n1 = Cast To Athena_PlayerController!(execute, Object=GetOwningPlayer()) -> (then, CastFailed, AsAthena Player Controller:Athena Player Controller Object Reference) {is=title K2Node_DynamicCast}
   n2 = Get ConvertComponent!(self:Fort Player Pawn Object Reference=GetPlayerPawn!(self:Fort Player Controller Object Reference=@n1) -> (ReturnValue:Fort Player Pawn Object Reference)) -> (ConvertComponent:Fort Pawn Component Convert Object Reference)
   > Is Valid(InputObject=@n2)
   n3 = GetConvertedPawns!(self:Fort Pawn Component Convert Object Reference=@n2, SpecificConvertedType:EConvertedType Enum=NPC) -> (ReturnValue:Array of Fort Pawn Object References)
   > Branch(Condition="integer > integer"(A=Length(TargetArray:Array of Fort Pawn Object References=@n3)))
   n4 = Get PlayerState!(self:Pawn Object Reference="Get (a copy)"(
     Array:Array of Fort Pawn Object References=@n3
     Dimension 1=Last Index!(TargetArray:Array of Fort Pawn Object References=@n3) -> (ReturnValue=0)
    ) -> (Output:Fort Pawn Object Reference)) -> (PlayerState:Player State Object Reference)
   > n5 = Is Valid(InputObject=@n4)
   Return Node!(execute=@n5.Is Not Valid, NpcPlayerState:Player State Object Reference)
   > Return Node(NpcPlayerState=@n4)
  ]}
]
variables [
 CurrentHiredNpcPlayerState {type=PlayerState, default=null}
 DetailsText {type=string, default=is now following you.}
 IconMaterial {type=MaterialInstance, default=/DynamicAthenaHUD/Materials/MI_UI_NpcHiredIcon}
 SceneReference {type=DynamicUIScene, default=/DynamicAthenaHUD/Scenes/AthenaHud_PlayerMessages}
]
widgets {
 info {
  inheritedWidgetCount=0
  parentClass=/Game/Athena/HUD/Messaging/WBP_NotificationWithBackplate_Visuals.WBP_NotificationWithBackplate_Visuals_C
  rootWidgetClass=null
  widgetCount=0
 }
 widgets []
}
